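# Two-sided lit material for the POI claim banner, showing a configurable icon on colored cloth that unfurls and moves in wind.
# An offset, scalable icon mask blends a two-color vertical gradient and surface values between banner and icon settings.
# World position offset gathers vertices toward the pivot height with sine folds, driven by BannerUnfurling or a time-sine preview switch.
# The shared Cloth function adds panning normal detail and offset, both scaled by WindMult.
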Material /POIBanner/Materials/M_POI_Claim_Banner
graph {
 properties {pixelDepthOffsetMode=PDOM_Legacy, previewMesh=/POIBanner/Meshes/POI_Claim_Banner, twoSided=true, usedShadingModels=Default Lit}
 nodes [
  ActorPositionWS()
  n1 = TextureCoordinate() {uTiling=0.4, vTiling=0.6}
  Add(A=MaterialFunctionCall(Float2=@n1) {materialFunction=&BreakOutFloat2Components}.G) {constB=0.5}
  Constant() {r=0.5}
  n2 = Add(
   A=TextureCoordinate()
   B=AppendVector(A=ScalarParameter() {parameterName=IconOffsetX}, B=ScalarParameter() {defaultValue=0.1, parameterName=IconOffsetY})
  )
  n3 = AppendVector(A=ScalarParameter() {defaultValue=0.8, parameterName=IconScaleX}, B=ScalarParameter() {defaultValue=0.55, parameterName=IconScaleY})
  Multiply(A=@n2, B=@n3)
  n4 = TextureCoordinate()
  n5 = VectorParameter() {parameterName=Normal Translation, defaultValue=<rgba 1 1 0 0>}
  n6 = AppendVector(A=@n5.R, B=@n5.R)
  n7 = TextureSampleParameter2D(UVs=Add(A=Multiply(A=@n4, B=@n6), B=AppendVector(A=@n5.B, B=@n5.B))) {parameterName=Normal, samplerType=SAMPLERTYPE_Normal, texture=/POIBanner/Textures/T_Gadget_POI_Claim_Banner_N}
  n8 = MaterialFunctionCall!(
   UVs=@n1
   Pan Speed=ScalarParameter() {defaultValue=-0.45, parameterName=ClothPanSpeed}
   WPO_ClothOffsetDirection
   WPO_ClothNormalIntensity
   WPO_ClothOffsetIntensity
   "Motion Mask (Defaults to the Red Vertex Color)"
   WPO_ClothWPOWindTextureOffset
   Pan Direction
  ) {materialFunction=/Game/Packages/Fortress_SharedMaterials/Base_Material_Functions/Cloth}
  n9 = ScalarParameter() {defaultValue=0.5, parameterName=AdditionalClothNormal}
  n10 = ScalarParameter() {parameterName=WindMult}
  NamedRerouteDeclaration(None=LinearInterpolate(
    A=@n7.RGB
    B=MaterialFunctionCall!(Normal_1=@n7.RGB, Normal_2=@n8.Normal) {materialFunction=/Game/Global/MaterialFunctions/MF_NormalBlend_Whiteout}.Result
    Alpha=Multiply(A=@n9, B=@n10)
   )) {name=Normal, nodeColor=<rgba 0 1 0.25 1>}
  n11 = Constant()
  n12 = Subtract(
   A=ComponentMask(None=MaterialFunctionCall() {materialFunction=&ObjectPivotPoint}.Object Pivot Location) {b=true}
   B=ComponentMask(None=WorldPosition().XYZ) {b=true}
  )
  n13 = TextureCoordinate()
  n14 = MaterialFunctionCall(Float2=@n13) {materialFunction=&BreakOutFloat2Components}
  n15 = SphereMask(A=TextureCoordinate(), B=Constant2Vector() {r=0.5}) {attenuationRadius=0.5, hardnessPercent=0}
  n16 = Clamp(None=Multiply(A=OneMinus(None=@n14.G), B=NamedRerouteDeclaration(None=@n15) {name=WPOMask, nodeColor=<rgba 1 0.6328 0 1>})) {minDefault=0.2}
  n17 = MaterialFunctionCall(X=@n11, Y=@n11, Z=Multiply(A=ComponentMask(None=@n12) {b=true}, B=Power(Base=@n16) {constExponent=0.1})) {materialFunction=&MakeFloat3}
  n18 = StaticSwitchParameter(
   True=Power(Base=MaterialFunctionCall() {materialFunction=&DebugTimeSine}.Result)
   False=OneMinus(None=ScalarParameter() {parameterName=BannerUnfurling})
  ) {parameterName=PreviewUnfurling}
  n19 = Constant2Vector() {g=0.5}
  n20 = Multiply(A=Power(Base=Add(A=@n13, B=@n19)) {constExponent=1}) {constB=4}
  n21 = ConstantBiasScale(None=Sine(None=MaterialFunctionCall(Float2=@n20) {materialFunction=&BreakOutFloat2Components}.G))
  n22 = If(A=@n21, "A > B"=Power(Base=@n21) {constExponent=0.5}, "A < B"=Power(Base=@n21)) {constB=0.5}
  n23 = ScalarParameter() {defaultValue=8, parameterName=Offset}
  n24 = StaticSwitchParameter(
   True=Power(Base=MaterialFunctionCall() {materialFunction=&DebugTimeSine}.Result)
   False=OneMinus(None=ScalarParameter() {parameterName=BannerUnfurling})
  ) {parameterName=PreviewUnfurling}
  n25 = Multiply(A=Constant3Vector() {constant=<rgba 0 1 0 0>}, B=Clamp(None=Multiply(A=@n23, B=@n24)) {maxDefault=20, minDefault=0.5})
  n26 = Transform(None=Add(A=Multiply(A=@n17.Result, B=@n18) {constB=0.6}, B=Multiply(A=@n22, B=@n25))) {transformSourceType=TRANSFORMSOURCE_Local}
  n27 = MaterialFunctionCall(Float2=TextureCoordinate()) {materialFunction=&BreakOutFloat2Components}
  n28 = Constant() {r=0.5}
  n29 = LinearInterpolate(
   A=@n26
   B=Multiply(A=@n26) {constB=0.65}
   Alpha=Power(Base=SphereMask(A=@n27.R, B=@n28) {attenuationRadius=0.65, hardnessPercent=0}) {constExponent=0.5}
  ) {constB=0}
  n30 = Multiply(A=ScalarParameter() {defaultValue=0.8, parameterName=AdditionalClothWPO}, B=ScalarParameter() {parameterName=WindMult})
  n31 = Multiply(A=@n8.Additive WPO, B=@n30)
  n32 = Transform(None=Constant3Vector() {desc="compensate wind wpo ", constant=<rgba -8 0 0 0>})
  NamedRerouteDeclaration(None=Add(A=Add(A=@n29, B=@n31), B=Multiply(A=@n32, B=@n30))) {name=wpo, nodeColor=<rgba 0 0.5547 1 1>}
  n33 = MaterialFunctionCall(UVs=@n2, Texture Scale=@n3) {materialFunction=&ScaleUVsByCenter}
  n34 = TextureSampleParameter2D(UVs=@n33.UVs) {parameterName=IconTexture, texture=/BRCosmetics/2dAssets/Banners/Season10/T-Banners-Icons-S10-80sLlama-L}
  NamedRerouteDeclaration(None=LinearInterpolate(A=Multiply(A=@n34.R, B=@n33.0-1 mask), B=Constant(), Alpha=ScalarParameter() {parameterName=HideIcon})) {name=IconMask, nodeColor=<rgba 0 1 0.6719 1>}
  n35 = VectorParameter() {parameterName=Color_Banner, defaultValue=<rgba 0.3194 0.2259 0.9427 1>}
  n36 = VectorParameter() {parameterName=Color_Banner2, defaultValue=<rgba 0.0839 0.0371 0.3958 1>}
  n37 = MaterialFunctionCall(Float2=TextureCoordinate()) {materialFunction=&BreakOutFloat2Components}
  NamedRerouteDeclaration(None=LinearInterpolate(
    A=LinearInterpolate(A=@n35.RGB, B=@n36.RGB, Alpha=@n37.G)
    B=VectorParameter() {parameterName=Color_Icon, defaultValue=<rgba 1 1 1 1>}.RGB
    Alpha=NamedRerouteUsage()
   )) {name=BaseColor, nodeColor=<rgba 0 1 0.8359 1>}
  NamedRerouteDeclaration(None=LinearInterpolate(
    A=ScalarParameter() {defaultValue=0.8, parameterName=Roughness_Banner}
    B=ScalarParameter() {defaultValue=0.25, parameterName=Roughness_Icon}
    Alpha=NamedRerouteUsage()
   )) {name=Rough, nodeColor=<rgba 0.0078 0 1 1>}
  NamedRerouteDeclaration(None=LinearInterpolate(
    A=ScalarParameter() {parameterName=Metal_Banner}
    B=ScalarParameter() {defaultValue=0.1, parameterName=Metal_Icon}
    Alpha=NamedRerouteUsage()
   )) {name=Metal, nodeColor=<rgba 0.9922 0 1 1>}
  NamedRerouteDeclaration(None=LinearInterpolate(
    A=ScalarParameter() {defaultValue=0.5, parameterName=Specular_Banner}
    B=ScalarParameter() {defaultValue=0.8, parameterName=Specular_Icon}
    Alpha=NamedRerouteUsage()
   )) {name=Spec, nodeColor=<rgba 0 0.9297 1 1>}
  NamedRerouteUsage()
  NamedRerouteUsage()
  NamedRerouteUsage()
  NamedRerouteUsage()
  NamedRerouteUsage()
  NamedRerouteUsage()
  NamedRerouteUsage()
  TextureSample() {samplerType=SAMPLERTYPE_Masks, texture=/POIBanner/Textures/T_Gadget_POI_Claim_Banner_S}
 ]
}
