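# Sound cue for a point-of-interest banner capture, playing a friendly or enemy stinger based on the listener's affiliation.
# A Switch on the integer parameter PlayerAffliation routes its first three inputs to the friendly mix and only the last to the enemy mix.
# Each mix layers a near and a distant wave, each through its own attenuation asset, so the sound changes with distance.
# A concurrency override caps simultaneous instances on the gadgets sound class.
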SoundCue /POIBanner/Sounds/POI_Captured_Cue
root=Switch_1
nodes [
 {id=Attenuation_0, children [WavePlayer_1], properties {attenuationSettings=&/POI_Capture_Point_Att}}
 {id=Attenuation_1, children [WavePlayer_3], properties {attenuationSettings=&POI_Capture_Point_Att}}
 {id=Attenuation_2, children [WavePlayer_5], properties {attenuationSettings=&/POI_Capture_Point_Distant_Att}}
 {id=Attenuation_3, children [WavePlayer_2], properties {attenuationSettings=&POI_Capture_Point_Distant_Att}}
 {id=Mixer_0, children [Attenuation_1, Attenuation_3], properties {inputVolume [1,1]}}
 {id=Mixer_1, children [Attenuation_0, Attenuation_2], properties {inputVolume [1.1,1]}}
 {id=Switch_1, children [Mixer_1,Mixer_1,Mixer_1,Mixer_0], properties {intParameterName=PlayerAffliation}}
 {id=WavePlayer_1, properties {soundWaveAssetPtr=&/POI_Captured_Friendly}}
 {id=WavePlayer_2, properties {soundWaveAssetPtr=&/POI_Captured_Enemy_Distant}}
 {id=WavePlayer_3, properties {soundWaveAssetPtr=&/POI_Captured_Enemy}}
 {id=WavePlayer_5, properties {soundWaveAssetPtr=&/POI_Captured_Friendly_Distant}}
]
properties {
 bOverrideConcurrency=true
 maxDistance=27500
 soundClassObject=&Fort_Gadgets
 volumeMultiplier=1.2
 concurrencyOverrides {maxCount=2}
 concurrencySet [/Game/Sounds/Concurrency_Rules/CONC_Discovery_POI]
}
