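# Sound cue for the pulse scan that plays when a POI banner capture point is captured and marks via scan.
# A mixer blends three layered waves at equal volume: a 2D report, a close layer and a distant layer.
# Each layer has its own attenuation asset, so the mix shifts with listener distance.
# It routes to the gadgets sound class and caps simultaneous plays with a concurrency override and a POI discovery concurrency set.
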SoundCue /POIBanner/Sounds/POI_Captured_Pulse_Marking_Scan_Cue
root=Mixer_0
nodes [
 {id=Attenuation_0, children [WavePlayer_1], properties {attenuationSettings=&/POI_Capture_Point_Pulse_Scan_2D_Att}}
 {id=Attenuation_5, children [WavePlayer_4], properties {attenuationSettings=&/POI_Capture_Point_Att}}
 {id=Attenuation_6, children [WavePlayer_0], properties {attenuationSettings=&/POI_Capture_Point_Distant_Att}}
 {id=Mixer_0, children [Attenuation_0, Attenuation_5, Attenuation_6], properties {inputVolume [1,1,1]}}
 {id=WavePlayer_0, properties {soundWaveAssetPtr=&/POI_Captured_Pulse_Scan_Distant_01}}
 {id=WavePlayer_1, properties {soundWaveAssetPtr=&/POI_Captured_Pulse_Scan_Reports_2D_01}}
 {id=WavePlayer_4, properties {soundWaveAssetPtr=&/POI_Captured_Pulse_Scan_Close_01}}
]
properties {
 bOverrideConcurrency=true
 maxDistance=27500
 soundClassObject=&Fort_Gadgets
 volumeMultiplier=1.2
 concurrencyOverrides {maxCount=2}
 concurrencySet [/Game/Sounds/Concurrency_Rules/CONC_Discovery_POI]
}
