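# Sound cue named as the stinger for capturing a point of interest in the POI banner feature.
# It has no root or sound nodes, so it produces no audio as it stands.
# It routes to the Fort_Gadgets sound class and overrides concurrency with its own small voice cap.
# That override takes precedence over the referenced CONC_Discovery_POI concurrency set.
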SoundCue /POIBanner/Sounds/POI_Captured_Stinger_Cue
root=null
nodes []
properties {
 bOverrideConcurrency=true
 maxDistance=2097152
 soundClassObject=&Fort_Gadgets
 volumeMultiplier=1
 concurrencyOverrides {maxCount=2}
 concurrencySet [/Game/Sounds/Concurrency_Rules/CONC_Discovery_POI]
}
