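# POI_Claim_Empty_Cue is a Sound Cue in the POIBanner content, named for a point-of-interest claim, that contains no nodes and plays no audio.
# Its root is unset, so only playback settings remain.
# These assign capture-point attenuation and the gadgets sound class, and override concurrency to cap simultaneous instances.
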SoundCue /POIBanner/Sounds/POI_Claim_Empty_Cue
root=null
nodes []
properties {
 attenuationSettings=&/POI_Capture_Point_Att
 bOverrideConcurrency=true
 maxDistance=18000
 soundClassObject=&Fort_Gadgets
 volumeMultiplier=1
 concurrencyOverrides {maxCount=2}
}
