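# Sound cue for a POI banner, named as the stop sound for the contested claim progress loop.
# It randomly picks one of four equally weighted stop whoosh waves, then a modulator raises it to nearly triple pitch with slight variation.
# The cue plays at reduced volume in the gadgets sound class, uses capture point attenuation, and overrides concurrency to allow a single instance.
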SoundCue /POIBanner/Sounds/POI_Claim_Progress_Contested_Loop_Stop_Cue
root=Modulator_1
nodes [
 {id=Modulator_1, children [Random_0], properties {pitchMax=3, pitchMin=2.8, volumeMax=1, volumeMin=1}}
 {
  id=Random_0
  children [WavePlayer_3,WavePlayer_4,WavePlayer_5,WavePlayer_6]
  properties {bShouldExcludeFromBranchCulling=true, weights [1,1,1,1]}
 }
 {id=WavePlayer_3, properties {soundWaveAssetPtr=&/POI_Claim_Progress_Stop_Whoosh_01}}
 {id=WavePlayer_4, properties {soundWaveAssetPtr=&/POI_Claim_Progress_Stop_Whoosh_02}}
 {id=WavePlayer_5, properties {soundWaveAssetPtr=&/POI_Claim_Progress_Stop_Whoosh_03}}
 {id=WavePlayer_6, properties {soundWaveAssetPtr=&/POI_Claim_Progress_Stop_Whoosh_04}}
]
properties {
 attenuationSettings=&/POI_Capture_Point_Att
 bOverrideConcurrency=true
 maxDistance=14500
 soundClassObject=&Fort_Gadgets
 volumeMultiplier=0.4
 concurrencyOverrides {maxCount=1}
}
