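# Sound cue for uncontested claim progress at a POI banner capture point, limited to one concurrent instance.
# It mixes a start sound, a looping bed and a randomized ratchet start.
# PlayerAffliation switches select enemy sounds on the last input and friendly or default sounds otherwise, with a shared ratchet loop.
# Loops fade in through envelopers, and ClaimPercent drives their pitch, plus bed volume, with separate pole and point attenuations.
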SoundCue /POIBanner/Sounds/POI_Claim_Progress_Uncontested_Loop_Cue
root=Mixer_4
nodes [
 {id=Attenuation_0, children [Modulator_0], properties {attenuationSettings=&/POI_Capture_Point_Pole_Att}}
 {id=Attenuation_1, children [Enveloper_0], properties {attenuationSettings=&POI_Capture_Point_Pole_Att}}
 {id=Attenuation_11, children [Modulator_5], properties {attenuationSettings=&/POI_Capture_Point_Att}}
 {id=Attenuation_13, children [Modulator_6], properties {attenuationSettings=&POI_Capture_Point_Att}}
 {id=Attenuation_2, children [Enveloper_1], properties {attenuationSettings=&POI_Capture_Point_Pole_Att}}
 {id=Attenuation_4, children [WavePlayer_2], properties {attenuationSettings=&POI_Capture_Point_Att}}
 {id=Attenuation_6, children [WavePlayer_4], properties {attenuationSettings=&POI_Capture_Point_Att}}
 {
  id=Enveloper_0
  children [WavePlayer_0]
  properties {pitchCurve {editorCurveData {keys [{time=0}]}}, volumeCurve {editorCurveData {keys [<at 0 0>,<at 1 1>]}}}
 }
 {
  id=Enveloper_1
  children [WavePlayer_1]
  properties {pitchCurve {editorCurveData {keys [{time=0}]}}, volumeCurve {editorCurveData {keys [<at 0 0>,<at 1 1>]}}}
 }
 {
  id=Enveloper_5
  children [WavePlayer_17]
  properties {pitchCurve {editorCurveData {keys [{time=0}]}}, volumeCurve {editorCurveData {keys [<at 0 0>,<at 1 1>]}}}
 }
 {
  id=Enveloper_6
  children [WavePlayer_20]
  properties {pitchCurve {editorCurveData {keys [{time=0}]}}, volumeCurve {editorCurveData {keys [<at 0 0>,<at 1 1>]}}}
 }
 {id=Mixer_0, children [ModulatorContinuous_0, ModulatorContinuous_1], properties {inputVolume [1,0.65]}}
 {id=Mixer_3, children [ModulatorContinuous_2, ModulatorContinuous_3], properties {inputVolume [1,0.95]}}
 {id=Mixer_4, children [Switch_1, Switch_0, Attenuation_0], properties {inputVolume [0.5,0.5,0.8]}}
 {
  id=ModulatorContinuous_0
  children [Attenuation_13]
  properties {
   pitchModulationParams {maxOutput=2, minOutput=0.8, parameterName=ClaimPercent}
   volumeModulationParams {minOutput=0.35, parameterName=ClaimPercent}
  }
 }
 {
  id=ModulatorContinuous_1
  children [Attenuation_2]
  properties {
   pitchModulationParams {maxOutput=1.5, minOutput=0.85, parameterName=ClaimPercent}
   volumeModulationParams {minOutput=1, parameterName=ClaimPercent}
  }
 }
 {
  id=ModulatorContinuous_2
  children [Attenuation_11]
  properties {
   pitchModulationParams {maxOutput=2, minOutput=0.8, parameterName=ClaimPercent}
   volumeModulationParams {minOutput=0.35, parameterName=ClaimPercent}
  }
 }
 {
  id=ModulatorContinuous_3
  children [Attenuation_1]
  properties {
   pitchModulationParams {maxOutput=1.5, minOutput=0.85, parameterName=ClaimPercent}
   volumeModulationParams {minOutput=1, parameterName=ClaimPercent}
  }
 }
 {id=Modulator_0, children [Random_0], properties {pitchMax=1.2, pitchMin=0.9}}
 {id=Modulator_5, children [Enveloper_5], properties {pitchMax=1, pitchMin=1, volumeMax=1, volumeMin=1}}
 {id=Modulator_6, children [Enveloper_6], properties {pitchMax=1, pitchMin=1, volumeMax=1, volumeMin=1}}
 {id=Random_0, children [WavePlayer_3,WavePlayer_5,WavePlayer_6], properties {bShouldExcludeFromBranchCulling=true, weights [1,1,1]}}
 {id=Switch_0, children [Mixer_0,Mixer_0,Mixer_0,Mixer_3], properties {intParameterName=PlayerAffliation}}
 {id=Switch_1, children [Attenuation_6, Attenuation_6, Attenuation_6, Attenuation_4], properties {intParameterName=PlayerAffliation}}
 {id=WavePlayer_0, properties {bLooping=true, soundWaveAssetPtr=&/POI_Claim_Progress_Uncontested_Ratchet_Loop}}
 {id=WavePlayer_1, properties {bLooping=true, soundWaveAssetPtr=&POI_Claim_Progress_Uncontested_Ratchet_Loop}}
 {id=WavePlayer_17, properties {bLooping=true, soundWaveAssetPtr=&/POI_Claim_Progress_Uncontested_Loop_Enemy}}
 {id=WavePlayer_2, properties {soundWaveAssetPtr=&/POI_Claim_Progress_Uncontested_Loop_Enemy_Start}}
 {id=WavePlayer_20, properties {bLooping=true, soundWaveAssetPtr=&/POI_Claim_Progress_Uncontested_Loop_Friendly}}
 {id=WavePlayer_3, properties {soundWaveAssetPtr=&/POI_Claim_Progress_Uncontested_Ratchet_Start_01}}
 {id=WavePlayer_4, properties {soundWaveAssetPtr=&/POI_Claim_Progress_Uncontested_Loop_Start}}
 {id=WavePlayer_5, properties {soundWaveAssetPtr=&/POI_Claim_Progress_Uncontested_Ratchet_Start_02}}
 {id=WavePlayer_6, properties {soundWaveAssetPtr=&/POI_Claim_Progress_Uncontested_Ratchet_Start_03}}
]
properties {
 bOverrideConcurrency=true
 maxDistance=14500
 soundClassObject=&Fort_Gadgets
 volumeMultiplier=1.2
 attenuationOverrides {falloffDistance=8000}
 concurrencyOverrides {maxCount=1}
}
