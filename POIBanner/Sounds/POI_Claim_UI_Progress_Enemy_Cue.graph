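# Sound cue for an enemy claim-progress UI sound in the POI Banner feature.
# A single wave player at the root plays the POI_Claim_UI_Uncontested_Enemy_01 wave directly, with no randomization or modulation nodes.
# It routes through the critical in-game UI sound class and overrides concurrency to one instance that stops the oldest, so a new trigger replaces the playing sound.
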SoundCue /POIBanner/Sounds/POI_Claim_UI_Progress_Enemy_Cue
root=WavePlayer_0
nodes [{id=WavePlayer_0, properties {soundWaveAssetPtr=&/POI_Claim_UI_Uncontested_Enemy_01}}]
properties {
 bOverrideConcurrency=true
 maxDistance=2097152
 soundClassObject=&Fort_UI_InGame_Critical
 volumeMultiplier=1
 concurrencyOverrides {maxCount=1, resolutionRule=StopOldest}
}
